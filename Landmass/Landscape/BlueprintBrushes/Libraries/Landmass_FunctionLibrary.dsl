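; Blueprint function library of shared helpers for Landmass landscape brushes and their brush manager.
; It converts world-space brush extents into pixel-snapped canvas coordinates, then draws a material or copies one render target into just that region of another, lazily creating the manager's copy material instance.
; Other helpers apply name-keyed parameter maps to material instances, fan-triangulate splines into landscape-space canvas triangles, and provide vector and extents math.
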(fn ApplyCustomMaterialParams
  (params (MID :type Material%20Instance%20Dynamic%20Object%20Reference) (Floats :type Map%20of%20Names%20to%20Float%20%28double-precision%29s) (Vectors :type Map%20of%20Names%20to%20Linear%20Color%20Structures) (Textures :type Map%20of%20Names%20to%20Texture%202D%20Object%20References))
  (bind _floats (Variables|GetFloats))
  (bind _vectors (Variables|GetVectors))
  (bind _textures (Variables|GetTextures))
  (Utilities|FlowControl|Sequence
    (:then_0
      (if (> (Utilities|Map|Length (Variables|GetFloats)) 0)
        (bind _keys (Utilities|Map|Keys (Variables|GetFloats)))
        (for _array_element _keys
          (if (.ReturnValue (Utilities|Map|Find _floats _array_element))
            (Rendering|Material|SetScalarParameterValue (Variables|GetMID) _array_element (Utilities|Map|Find _floats _array_element) :!class /Script/Engine.MaterialInstanceDynamic)))))
    (:then_1
      (if (> (Utilities|Map|Length (Variables|GetVectors)) 0)
        (bind _keys_2 (Utilities|Map|Keys (Variables|GetVectors)))
        (for _array_element_2 _keys_2
          (if (.ReturnValue (Utilities|Map|Find _vectors _array_element_2))
            (Rendering|Material|SetVectorParameterValue (Variables|GetMID) _array_element_2 (Utilities|Map|Find _vectors _array_element_2) :!class /Script/Engine.MaterialInstanceDynamic)))))
    (:then_2
      (if (> (Utilities|Map|Length (Variables|GetTextures)) 0)
        (bind _keys_3 (Utilities|Map|Keys (Variables|GetTextures)))
        (for _array_element_3 _keys_3
          (if (.ReturnValue (Utilities|Map|Find _textures _array_element_3))
            (Rendering|Material|SetTextureParameterValue (Variables|GetMID) _array_element_3 (Utilities|Map|Find _textures _array_element_3))))))))

(fn BreakExtents
  (params (Extents :type Vector%204%20Structure))
  (bind (_x _y _z _w) (Math|Vector4|BreakVector4 Extents))
  (return (Math|Vector2D|MakeVector2D _x _y) (Math|Vector2D|MakeVector2D _z _w)))

(fn BrushCopyRenderTargetRegion
  (params (Source%20RT :type Texture%20Render%20Target%202D%20Object%20Reference) (Destination%20RT :type Texture%20Render%20Target%202D%20Object%20Reference) (Brush%20Extents :type Vector%204%20Structure) (Manager :type Landmass%20Brush%20Manager%20Object%20Reference))
  (bind (_screen_position _screen_size _coordinate_position _coordinate_size) (Default|WorldExtentsToCanvasCoords (Variables|GetBrush%20Extents) (Variables|GetManager) (Variables|GetManager)))
  ; Ensure Copy RT MID is Initialized
  (Utilities|IsValid (Variables|/Landmass/Landscape/BlueprintBrushes/LandmassBrushManager.LandmassBrushManager_C|GetCopyRT_MID (Variables|GetManager))
    (:Is%20Valid
      _)
    (:Is%20Not%20Valid
      (bind _returnvalue (Rendering|Material|CreateDynamicMaterialInstance (Variables|GetManager) "/Landmass/Landscape/BlueprintBrushes/Materials/Internal/CopyRT_LandscapeUVs.CopyRT_LandscapeUVs" "None" "None" :!class /Script/Engine.KismetMaterialLibrary))
      (Variables|/Landmass/Landscape/BlueprintBrushes/LandmassBrushManager.LandmassBrushManager_C|SetCopyRT_MID _returnvalue (Variables|GetManager))))
  (Rendering|Material|SetTextureParameterValue (Variables|/Landmass/Landscape/BlueprintBrushes/LandmassBrushManager.LandmassBrushManager_C|GetCopyRT_MID (Variables|GetManager)) "RT" Source%20RT)
  (bind (_canvas _size _context) (Rendering|BeginDrawCanvastoRenderTarget (Variables|GetManager) Destination%20RT))
  (Canvas|DrawMaterial _canvas (Variables|/Landmass/Landscape/BlueprintBrushes/LandmassBrushManager.LandmassBrushManager_C|GetCopyRT_MID (Variables|GetManager)) _screen_position _screen_size _coordinate_position _coordinate_size 0.0 "(X=0.500000,Y=0.500000)")
  (Rendering|EndDrawCanvastoRenderTarget (Variables|GetManager) _context))

(fn BrushDrawtoRenderTargetRegion
  (params (RenderTarget :type Texture%20Render%20Target%202D%20Object%20Reference) (Material :type Material%20Interface%20Object%20Reference) (Brush%20Extents :type Vector%204%20Structure) (Manager :type Landmass%20Brush%20Manager%20Object%20Reference))
  (bind (_screen_position _screen_size _coordinate_position _coordinate_size) (Default|WorldExtentsToCanvasCoords (Variables|GetBrush%20Extents) (Variables|GetManager) (Variables|GetManager)))
  (bind (_canvas _size _context) (Rendering|BeginDrawCanvastoRenderTarget (Variables|GetManager) RenderTarget))
  (Canvas|DrawMaterial _canvas (Variables|GetMaterial) _screen_position _screen_size _coordinate_position _coordinate_size 0.0 "(X=0.500000,Y=0.500000)")
  (Rendering|EndDrawCanvastoRenderTarget (Variables|GetManager) _context))

(fn CeilV2
  (params (In :type Vector%202D%20Structure))
  (bind (_x _y) (Math|Vector2D|BreakVector2D In))
  (return (Math|Vector2D|MakeVector2D (Math|Conversions|ToFloat(Integer) (Math|Float|Ceil _x)) (Math|Conversions|ToFloat(Integer) (Math|Float|Ceil _y)))))

(fn FloorV2
  (params (In :type Vector%202D%20Structure))
  (bind (_x _y) (Math|Vector2D|BreakVector2D In))
  (return (Math|Vector2D|MakeVector2D (Math|Conversions|ToFloat(Integer) (Math|Float|Floor _x)) (Math|Conversions|ToFloat(Integer) (Math|Float|Floor _y)))))

(fn MakeExtents
  (params (Min :type Vector%202D%20Structure) (Max :type Vector%202D%20Structure))
  (bind (_x _y) (Math|Vector2D|BreakVector2D Max))
  (bind (_x_2 _y_2) (Math|Vector2D|BreakVector2D Min))
  (return (Math|Vector4|MakeVector4 _x_2 _y_2 _x _y)))

(fn MaxV2
  (params (A :type Vector%202D%20Structure) (B :type Vector%202D%20Structure))
  (bind (_x _y) (Math|Vector2D|BreakVector2D B))
  (bind (_x_2 _y_2) (Math|Vector2D|BreakVector2D A))
  (return (Math|Vector2D|MakeVector2D (Math|Float|Max(Float) _x_2 _x) (Math|Float|Max(Float) _y_2 _y))))

(fn MaxV3
  (params (A :type Vector) (B :type Vector))
  (return (Math|Vector|MakeVector (Math|Float|Max(Float) (.x A) (.x B)) (Math|Float|Max(Float) (.y A) (.y B)) (Math|Float|Max(Float) (.z A) (.z B)))))

(fn MinV2
  (params (A :type Vector%202D%20Structure) (B :type Vector%202D%20Structure))
  (bind (_x _y) (Math|Vector2D|BreakVector2D B))
  (bind (_x_2 _y_2) (Math|Vector2D|BreakVector2D A))
  (return (Math|Vector2D|MakeVector2D (Math|Float|Min(Float) _x_2 _x) (Math|Float|Min(Float) _y_2 _y))))

(fn MinV3
  (params (A :type Vector) (B :type Vector))
  (return (Math|Vector|MakeVector (Math|Float|Min(Float) (.x A) (.x B)) (Math|Float|Min(Float) (.y A) (.y B)) (Math|Float|Min(Float) (.z A) (.z B)))))

(fn PointSampleUV
  (params (In :type Vector%202D%20Structure) (Res :type Vector%202D%20Structure) (Ceil :type Boolean))
  (return (/ (select Ceil (Default|CeilV2 (* In Res)) (CallFunction|FloorV2 (* In Res))) Res)))

(fn SplinePointstoCanvasUVTris
  (params (Spline :type Spline%20Component%20Object%20Reference) (Landscape%20Info :type Landmass%20Landscape%20Info%20Structure) (FixedLength :type Float%20%28double-precision%29 :default 1e+02))
  (bind _returnvalue (Spline|GetSplineLength Spline))
  (bind _returnvalue_2 (/ _returnvalue FixedLength))
  (Utilities|Array|Clear (LocalVariables|GetCanvasUVTris))
  (if (== FixedLength 0)
    ; Spline Points
    (for _index (range (- (Spline|GetNumberOfSplinePoints Spline) 2))
      (Utilities|Array|Add (LocalVariables|GetCanvasUVTris) (Utilities|Struct|MakeCanvasUVTri (Math|Conversions|ToVector2D(Vector) (+ (Math|Transform|InverseTransformLocation (Utilities|Struct|BreakLandmassLandscapeInfo (Variables|GetLandscape%20Info)) (Spline|GetLocationatSplinePoint (Variables|GetSpline) 0 "World")) 0.5)) "(X=0.000000,Y=0.000000)" "(R=1.000000,G=0.000000,B=0.000000,A=1.000000)" (Math|Conversions|ToVector2D(Vector) (+ (Math|Transform|InverseTransformLocation (Utilities|Struct|BreakLandmassLandscapeInfo (Variables|GetLandscape%20Info)) (Spline|GetLocationatSplinePoint (Variables|GetSpline) (+ (+ _index 1) 1) "World")) 0.5)) "(X=0.000000,Y=0.000000)" "(R=0.000000,G=1.000000,B=0.245000,A=1.000000)" (Math|Conversions|ToVector2D(Vector) (+ (Math|Transform|InverseTransformLocation (Utilities|Struct|BreakLandmassLandscapeInfo (Variables|GetLandscape%20Info)) (Spline|GetLocationatSplinePoint (Variables|GetSpline) (+ _index 1) "World")) 0.5)) "(X=0.000000,Y=0.000000)" "(R=0.000306,G=0.000000,B=1.000000,A=1.000000)")))
    (return (LocalVariables|GetCanvasUVTris))
    (else
      ; Spline Points
      (for _index_2 (range (+ (- (Math|Float|Truncate _returnvalue_2) 2) 1))
        (Utilities|Array|Add (LocalVariables|GetCanvasUVTris) (Utilities|Struct|MakeCanvasUVTri (Math|Conversions|ToVector2D(Vector) (+ (Math|Transform|InverseTransformLocation (Utilities|Struct|BreakLandmassLandscapeInfo (Variables|GetLandscape%20Info)) (Spline|GetLocationatDistanceAlongSpline (Variables|GetSpline) 0.0 "World")) 0.5)) "(X=0.000000,Y=0.000000)" "(R=1.000000,G=0.000000,B=0.000000,A=1.000000)" (Math|Conversions|ToVector2D(Vector) (+ (Math|Transform|InverseTransformLocation (Utilities|Struct|BreakLandmassLandscapeInfo (Variables|GetLandscape%20Info)) (Spline|GetLocationatDistanceAlongSpline (Variables|GetSpline) (* (+ (+ _index_2 1) 1) (Variables|GetFixedLength)) "World")) 0.5)) "(X=0.000000,Y=0.000000)" "(R=0.000000,G=1.000000,B=0.245000,A=1.000000)" (Math|Conversions|ToVector2D(Vector) (+ (Math|Transform|InverseTransformLocation (Utilities|Struct|BreakLandmassLandscapeInfo (Variables|GetLandscape%20Info)) (Spline|GetLocationatDistanceAlongSpline (Variables|GetSpline) (* (+ _index_2 1) (Variables|GetFixedLength)) "World")) 0.5)) "(X=0.000000,Y=0.000000)" "(R=0.000306,G=0.000000,B=1.000000,A=1.000000)")))
      (Utilities|Array|Add (LocalVariables|GetCanvasUVTris) (Utilities|Struct|MakeCanvasUVTri (Math|Conversions|ToVector2D(Vector) (+ (Math|Transform|InverseTransformLocation (Utilities|Struct|BreakLandmassLandscapeInfo (Variables|GetLandscape%20Info)) (Spline|GetLocationatDistanceAlongSpline (Variables|GetSpline) 0.0 "World")) 0.5)) "(X=0.000000,Y=0.000000)" "(R=1.000000,G=0.000000,B=0.000000,A=1.000000)" (Math|Conversions|ToVector2D(Vector) (+ (Math|Transform|InverseTransformLocation (Utilities|Struct|BreakLandmassLandscapeInfo (Variables|GetLandscape%20Info)) (Spline|GetLocationatDistanceAlongSpline (Variables|GetSpline) (Spline|GetSplineLength (Variables|GetSpline)) "World")) 0.5)) "(X=0.000000,Y=0.000000)" "(R=0.000000,G=1.000000,B=0.245000,A=1.000000)" (Math|Conversions|ToVector2D(Vector) (+ (Math|Transform|InverseTransformLocation (Utilities|Struct|BreakLandmassLandscapeInfo (Variables|GetLandscape%20Info)) (Spline|GetLocationatDistanceAlongSpline (Variables|GetSpline) (- (Spline|GetSplineLength (Variables|GetSpline)) (* (Math|Float|Fraction _returnvalue_2) (Variables|GetFixedLength))) "World")) 0.5)) "(X=0.000000,Y=0.000000)" "(R=0.000306,G=0.000000,B=1.000000,A=1.000000)"))
      (return (LocalVariables|GetCanvasUVTris)))))

(fn WorldExtentstoCanvasCoords
  (params (World%20Extents :type Vector%204%20Structure) (Manager :type Landmass%20Brush%20Manager%20Object%20Reference))
  (bind (_x _y) (Math|Vector4|BreakVector4 (Variables|GetWorld%20Extents)))
  (bind (_x_2 _y_2 _z _w) (Math|Vector4|BreakVector4 (Variables|GetWorld%20Extents)))
  (return (+ (CallFunction|FloorV2 (* (Variables|/Landmass/Landscape/BlueprintBrushes/LandmassBrushManager.LandmassBrushManager_C|GetLandscape%20RT%20Res (Variables|GetManager)) (/ (- (Math|Vector2D|MakeVector2D _x _y) (Math|Conversions|ToVector2D(Vector) (.location (Variables|/Landmass/Landscape/BlueprintBrushes/LandmassBrushManager.LandmassBrushManager_C|GetLS%20Transform (Variables|GetManager))))) (Variables|/Landmass/Landscape/BlueprintBrushes/LandmassBrushManager.LandmassBrushManager_C|GetRT%20World%20Size (Variables|GetManager))))) "(X=0.000000,Y=0.000000)") (+ (Default|CeilV2 (+ (* (Variables|/Landmass/Landscape/BlueprintBrushes/LandmassBrushManager.LandmassBrushManager_C|GetLandscape%20RT%20Res (Variables|GetManager)) (/ (- (Math|Vector2D|MakeVector2D _z _w) (Math|Vector2D|MakeVector2D _x_2 _y_2)) (Variables|/Landmass/Landscape/BlueprintBrushes/LandmassBrushManager.LandmassBrushManager_C|GetRT%20World%20Size (Variables|GetManager)))) "(X=1.000000,Y=1.000000)")) "(X=0.000000,Y=0.000000)") (/ (+ (CallFunction|FloorV2 (* (Variables|/Landmass/Landscape/BlueprintBrushes/LandmassBrushManager.LandmassBrushManager_C|GetLandscape%20RT%20Res (Variables|GetManager)) (/ (- (Math|Vector2D|MakeVector2D _x _y) (Math|Conversions|ToVector2D(Vector) (.location (Variables|/Landmass/Landscape/BlueprintBrushes/LandmassBrushManager.LandmassBrushManager_C|GetLS%20Transform (Variables|GetManager))))) (Variables|/Landmass/Landscape/BlueprintBrushes/LandmassBrushManager.LandmassBrushManager_C|GetRT%20World%20Size (Variables|GetManager))))) "(X=0.000000,Y=0.000000)") (Math|Conversions|ToVector2D(IntPoint) (Variables|/Landmass/Landscape/BlueprintBrushes/LandmassBrushManager.LandmassBrushManager_C|GetLandscape%20RT%20Res (Variables|GetManager)))) (/ (+ (Default|CeilV2 (+ (* (Variables|/Landmass/Landscape/BlueprintBrushes/LandmassBrushManager.LandmassBrushManager_C|GetLandscape%20RT%20Res (Variables|GetManager)) (/ (- (Math|Vector2D|MakeVector2D _z _w) (Math|Vector2D|MakeVector2D _x_2 _y_2)) (Variables|/Landmass/Landscape/BlueprintBrushes/LandmassBrushManager.LandmassBrushManager_C|GetRT%20World%20Size (Variables|GetManager)))) "(X=1.000000,Y=1.000000)")) "(X=0.000000,Y=0.000000)") (Math|Conversions|ToVector2D(IntPoint) (Variables|/Landmass/Landscape/BlueprintBrushes/LandmassBrushManager.LandmassBrushManager_C|GetLandscape%20RT%20Res (Variables|GetManager))))))
